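# Material function for the ZebraScramble back bling screen, playing a looping animated sequence with TV glitches.
# Time drives five clips with tunable range, speed and loop count, yielding a frame index that selects a cell across four flipbook textures.
# Panning static and a sweep band distort UVs, while SphereMask-based blends add static bursts at clip transitions.
# Desaturation, constant static and vignette finish the output; an extra single-atlas sample is unconnected.
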MaterialFunction /BRCosmetics/Accessories/FORT_Backpacks/Backpack_M_MED_ZebraScramble_Bacon/Materials/MF_ZebraScrambleScreen_V2
graph {nodes [
  n1 = SphereMask(
   A=NamedRerouteUsage()
   B=Constant2Vector() {g=0.5, r=0.5}
   Radius=ScalarParameter() {defaultValue=0.5, group=ZebraScramble, parameterName=ZS_VignetteRadius, sortPriority=100}
   Hardness=ScalarParameter() {defaultValue=0.3, group=ZebraScramble, parameterName=ZS_VignetteHardness, sortPriority=101}
  ) {attenuationRadius=0.55, hardnessPercent=40}
  n2 = ScalarParameter() {defaultValue=1, group=ZebraScramble, parameterName=ZS_VignetteStrength, sortPriority=102}
  n3 = NamedRerouteUsage()
  n4 = Constant() {r=3}
  n5 = ScalarParameter() {defaultValue=4, group=ZebraScramble, parameterName=ZS_FlipbookColumns, sortPriority=2}
  n6 = AppendVector(A=@n5, B=ScalarParameter() {defaultValue=4, group=ZebraScramble, parameterName=ZS_FlipbookRows, sortPriority=1})
  n7 = Divide(A=NamedRerouteUsage(), B=@n6)
  n8 = NamedRerouteUsage()
  n9 = Fmod(A=@n8, B=@n5)
  n10 = Divide(A=@n8, B=@n5)
  n11 = Divide(A=AppendVector(A=Floor(None=@n9), B=Floor(None=@n10)), B=@n6)
  n12 = TextureSampleParameter2D(UVs=Add(A=Add(A=@n7, B=@n11), B=NamedRerouteUsage())) {
   group=ZebraScramble
   parameterName=ZS_Flipbook_4
   sortPriority=0
   texture=/BRCosmetics/Accessories/FORT_Backpacks/Backpack_M_MED_ZebraScramble_Bacon/Textures/M_FrenchFry_screen_v3_4
  }
  n13 = ScalarParameter() {defaultValue=4, group=ZebraScramble, parameterName=ZS_FlipbookColumns, sortPriority=2}
  n14 = AppendVector(A=@n13, B=ScalarParameter() {defaultValue=4, group=ZebraScramble, parameterName=ZS_FlipbookRows, sortPriority=1})
  n15 = Divide(A=NamedRerouteUsage(), B=@n14)
  n16 = NamedRerouteUsage()
  n17 = Fmod(A=@n16, B=@n13)
  n18 = Divide(A=@n16, B=@n13)
  n19 = Divide(A=AppendVector(A=Floor(None=@n17), B=Floor(None=@n18)), B=@n14)
  n20 = TextureSampleParameter2D(UVs=Add(A=Add(A=@n15, B=@n19), B=NamedRerouteUsage())) {
   group=ZebraScramble
   parameterName=ZS_Flipbook_3
   sortPriority=0
   texture=/BRCosmetics/Accessories/FORT_Backpacks/Backpack_M_MED_ZebraScramble_Bacon/Textures/M_FrenchFry_screen_v3_3
  }
  n21 = ScalarParameter() {defaultValue=4, group=ZebraScramble, parameterName=ZS_FlipbookColumns, sortPriority=2}
  n22 = AppendVector(A=@n21, B=ScalarParameter() {defaultValue=4, group=ZebraScramble, parameterName=ZS_FlipbookRows, sortPriority=1})
  n23 = Divide(A=NamedRerouteUsage(), B=@n22)
  n24 = NamedRerouteUsage()
  n25 = Fmod(A=@n24, B=@n21)
  n26 = Divide(A=@n24, B=@n21)
  n27 = Divide(A=AppendVector(A=Floor(None=@n25), B=Floor(None=@n26)), B=@n22)
  n28 = TextureSampleParameter2D(UVs=Add(A=Add(A=@n23, B=@n27), B=NamedRerouteUsage())) {
   group=ZebraScramble
   parameterName=ZS_Flipbook_2
   sortPriority=0
   texture=/BRCosmetics/Accessories/FORT_Backpacks/Backpack_M_MED_ZebraScramble_Bacon/Textures/M_FrenchFry_screen_v3_2
  }
  n29 = ScalarParameter() {defaultValue=4, group=ZebraScramble, parameterName=ZS_FlipbookColumns, sortPriority=2}
  n30 = AppendVector(A=@n29, B=ScalarParameter() {defaultValue=4, group=ZebraScramble, parameterName=ZS_FlipbookRows, sortPriority=1})
  n31 = Divide(A=NamedRerouteUsage(), B=@n30)
  n32 = NamedRerouteUsage()
  n33 = Fmod(A=@n32, B=@n29)
  n34 = Divide(A=@n32, B=@n29)
  n35 = Divide(A=AppendVector(A=Floor(None=@n33), B=Floor(None=@n34)), B=@n30)
  n36 = TextureSampleParameter2D(UVs=Add(A=Add(A=@n31, B=@n35), B=NamedRerouteUsage())) {
   group=ZebraScramble
   parameterName=ZS_Flipbook_1
   sortPriority=0
   texture=/BRCosmetics/Accessories/FORT_Backpacks/Backpack_M_MED_ZebraScramble_Bacon/Textures/M_FrenchFry_screen_v3_1
  }
  n37 = If(
   A=NamedRerouteUsage()
   B=Constant() {r=2}
   "A > B"=@n20.RGB
   "A == B"=@n20.RGB
   "A < B"=If(A=NamedRerouteUsage(), B=Constant() {r=1}, "A > B"=@n28.RGB, "A == B"=@n28.RGB, "A < B"=@n36.RGB)
  )
  n38 = LinearInterpolate(
   A=Desaturation(
    None=If(A=@n3, B=@n4, "A > B"=@n12.RGB, "A == B"=@n12.RGB, "A < B"=@n37)
    Fraction=ScalarParameter() {group=ZebraScramble, parameterName=ZS_Desaturation, sortPriority=100}
   )
   B=NamedRerouteUsage()
   Alpha=NamedRerouteUsage()
  )
  n39 = NamedRerouteUsage()
  FunctionOutput(None=Multiply(A=LinearInterpolate(B=@n1, Alpha=@n2) {constA=1}, B=Add(A=@n38, B=@n39)))
  n40 = NamedRerouteUsage()
  n41 = TextureSample(UVs=Panner(Coordinate=@n40) {speedX=6, speedY=-0.05}) {
   samplerSource=SSM_Wrap_WorldGroupSettings
   samplerType=SAMPLERTYPE_Masks
   texture=/Game/Missions/Primary/RepairTheShelter/Textures/Cinematic_Screen_Static
  }
  n42 = TextureSample(UVs=Panner(Coordinate=@n40) {speedX=-3, speedY=-0.03}) {
   samplerSource=SSM_Wrap_WorldGroupSettings
   samplerType=SAMPLERTYPE_Masks
   texture=&Cinematic_Screen_Static
  }
  n43 = Multiply(A=@n41.G, B=@n42.R)
  n44 = Multiply(A=@n43, B=ScalarParameter() {defaultValue=0.03, group=TV Glitches, parameterName=Static UV Distort})
  n45 = Constant()
  n46 = NamedRerouteUsage()
  n47 = Multiply(A=OneMinus(None=ComponentMask(None=@n46) {g=true})) {constB=0.75}
  n48 = Multiply(A=Time(), B=ScalarParameter() {defaultValue=-0.2, group=TV Glitches, parameterName=Sweep Speed}) {constB=-0.2}
  n49 = Power(Base=Frac(None=Add(A=@n47, B=@n48))) {constExponent=20}
  n50 = Multiply(
   A=@n49
   B=LinearInterpolate(
    A=ScalarParameter() {defaultValue=0.3, group=TV Glitches, parameterName=Sweep Add Base}
    B=ScalarParameter() {defaultValue=0.3, group=TV Glitches, parameterName=Sweep Add}
   ) {constAlpha=1}
  ) {constB=0.2}
  n51 = AppendVector(A=Multiply(A=@n49, B=ScalarParameter() {defaultValue=0.01, group=TV Glitches, parameterName=Sweep UV Distort}) {constB=0.02}, B=Constant())
  NamedRerouteDeclaration(None=Add(A=AppendVector(A=@n44, B=@n45), B=Multiply(A=@n50, B=@n51))) {name=StaticUVDistort, nodeColor=<rgba 0 1 0.1094 1>}
  NamedRerouteDeclaration(None=ScalarParameter() {group=ZebraScramble, parameterName=1_ClipStart, sortPriority=10}) {name=1_ClipStart, nodeColor=<rgba 0 0.5781 1 1>}
  n52 = TextureSample(UVs=Panner(Coordinate=@n40) {speedX=5, speedY=0.1}) {
   samplerSource=SSM_Wrap_WorldGroupSettings
   samplerType=SAMPLERTYPE_Masks
   texture=&Cinematic_Screen_Static
  }
  n53 = Multiply(A=Add(A=Multiply(A=@n52.B) {constB=0.1}, B=@n43), B=ScalarParameter() {defaultValue=3, group=TV Glitches, parameterName=Static Scale})
  NamedRerouteDeclaration(None=Multiply(A=NamedRerouteDeclaration(None=@n53) {name=Static, nodeColor=<rgba 1 0.6562 0 1>}) {constB=0.2, desc=constant static}) {name=StaticConstant, nodeColor=<rgba 0 0.7422 1 1>}
  NamedRerouteDeclaration(None=ScalarParameter() {defaultValue=31, group=ZebraScramble, parameterName=2_ClipEnd, sortPriority=21}) {name=2_ClipEnd, nodeColor=<rgba 0 1 0.3672 1>}
  n54 = NamedRerouteUsage()
  NamedRerouteDeclaration(None=Subtract(A=Add(A=@n54), B=NamedRerouteUsage())) {name=2_ClipRange, nodeColor=<rgba 1 0.7734 0 1>}
  NamedRerouteDeclaration(None=ScalarParameter() {defaultValue=5, group=ZebraScramble, parameterName=2_ClipSpeed, sortPriority=22}) {name=2_ClipSpeed, nodeColor=<rgba 0 0.5781 1 1>}
  NamedRerouteDeclaration(None=Multiply(A=NamedRerouteUsage(), B=NamedRerouteUsage())) {name=2_ClipTime, nodeColor=<rgba 1 0 0.8203 1>}
  n55 = Multiply(A=NamedRerouteUsage(), B=NamedRerouteUsage())
  NamedRerouteDeclaration(None=Divide(A=NamedRerouteDeclaration(None=@n55) {name=1_ClipLength, nodeColor=<rgba 0.7578 1 0 1>}, B=NamedRerouteUsage())) {name=1_ClipTimeLength, nodeColor=<rgba 1 0 0.5625 1>}
  n56 = Multiply(A=NamedRerouteUsage(), B=ScalarParameter() {defaultValue=3, group=ZebraScramble, parameterName=2_ClipLoop, sortPriority=23})
  NamedRerouteDeclaration(None=Divide(A=NamedRerouteDeclaration(None=@n56) {name=2_ClipLength, nodeColor=<rgba 0 1 1 1>}, B=NamedRerouteUsage())) {name=2_ClipTimeLength, nodeColor=<rgba 1 0 0.5859 1>}
  n57 = NamedRerouteDeclaration(None=Add(A=NamedRerouteUsage(), B=NamedRerouteUsage())) {name=2_ClipTimeEnd, nodeColor=<rgba 0 1 0.2734 1>}
  n58 = NamedRerouteUsage()
  n59 = Add(A=NamedRerouteDeclaration(None=Add(A=@n57, B=@n58)) {name=3_ClipTimeEnd, nodeColor=<rgba 1 0.1406 0 1>}, B=NamedRerouteUsage())
  NamedRerouteDeclaration(None=Add(A=NamedRerouteDeclaration(None=@n59) {name=4_ClipTimeEnd, nodeColor=<rgba 0.6641 1 0 1>}, B=NamedRerouteUsage())) {name=TotalTimeLength, nodeColor=<rgba 0 0.6719 1 1>}
  NamedRerouteDeclaration(None=Fmod(A=Time(), B=NamedRerouteUsage())) {name=TimeLoop, nodeColor=<rgba 0 0.3906 1 1>}
  NamedRerouteDeclaration(None=ScalarParameter() {defaultValue=4, group=ZebraScramble, parameterName=1_ClipEnd, sortPriority=11}) {name=1_ClipEnd, nodeColor=<rgba 0 0.5781 1 1>}
  NamedRerouteDeclaration(None=ScalarParameter() {defaultValue=12, group=ZebraScramble, parameterName=3_ClipStart, sortPriority=30}) {name=3_ClipStart, nodeColor=<rgba 0 1 0.3672 1>}
  NamedRerouteDeclaration(None=ScalarParameter() {defaultValue=19, group=ZebraScramble, parameterName=3_ClipEnd, sortPriority=31}) {name=3_ClipEnd, nodeColor=<rgba 0 1 0.3672 1>}
  NamedRerouteDeclaration(None=ScalarParameter() {defaultValue=4, group=ZebraScramble, parameterName=3_ClipSpeed}) {name=3_ClipSpeed, nodeColor=<rgba 0 0.5781 1 1>}
  n60 = NamedRerouteUsage()
  NamedRerouteDeclaration(None=Subtract(A=Add(A=@n60), B=NamedRerouteUsage())) {name=3_ClipRange, nodeColor=<rgba 1 0.7734 0 1>}
  n61 = Multiply(A=NamedRerouteUsage(), B=ScalarParameter() {defaultValue=4, group=ZebraScramble, parameterName=3_ClipLoop, sortPriority=33})
  NamedRerouteDeclaration(None=Divide(A=NamedRerouteDeclaration(None=@n61) {name=3_ClipLength, nodeColor=<rgba 0 1 1 1>}, B=NamedRerouteUsage())) {name=3_ClipTimeLength, nodeColor=<rgba 1 0 0.5859 1>}
  NamedRerouteDeclaration(None=Multiply(A=NamedRerouteUsage(), B=NamedRerouteUsage())) {name=3_ClipTime, nodeColor=<rgba 1 0 0.8203 1>}
  NamedRerouteDeclaration(None=ScalarParameter() {defaultValue=4, group=ZebraScramble, parameterName=1_ClipSpeed, sortPriority=12}) {name=1_ClipSpeed, nodeColor=<rgba 0 0.5781 1 1>}
  NamedRerouteDeclaration(None=ScalarParameter() {defaultValue=32, group=ZebraScramble, parameterName=4_ClipStart, sortPriority=40}) {name=4_ClipStart, nodeColor=<rgba 0 1 0.3672 1>}
  NamedRerouteDeclaration(None=ScalarParameter() {defaultValue=63, group=ZebraScramble, parameterName=4_ClipEnd, sortPriority=41}) {name=4_ClipEnd, nodeColor=<rgba 0 1 0.3672 1>}
  NamedRerouteDeclaration(None=ScalarParameter() {defaultValue=3, group=ZebraScramble, parameterName=4_ClipSpeed, sortPriority=42}) {name=4_ClipSpeed, nodeColor=<rgba 0 0.5781 1 1>}
  n62 = NamedRerouteUsage()
  NamedRerouteDeclaration(None=Subtract(A=Add(A=@n62), B=NamedRerouteUsage())) {name=4_ClipRange, nodeColor=<rgba 1 0.7734 0 1>}
  n63 = Multiply(A=NamedRerouteUsage(), B=ScalarParameter() {defaultValue=1, group=ZebraScramble, parameterName=4_ClipLoop, sortPriority=43})
  NamedRerouteDeclaration(None=Divide(A=NamedRerouteDeclaration(None=@n63) {name=4_ClipLength, nodeColor=<rgba 0 1 1 1>}, B=NamedRerouteUsage())) {name=4_ClipTimeLength, nodeColor=<rgba 1 0 0.5859 1>}
  NamedRerouteDeclaration(None=ScalarParameter() {defaultValue=5, group=ZebraScramble, parameterName=5_ClipStart, sortPriority=50}) {name=5_ClipStart, nodeColor=<rgba 0 1 0.3672 1>}
  NamedRerouteDeclaration(None=ScalarParameter() {defaultValue=11, group=ZebraScramble, parameterName=5_ClipEnd, sortPriority=51}) {name=5_ClipEnd, nodeColor=<rgba 0 1 0.3672 1>}
  NamedRerouteDeclaration(None=ScalarParameter() {defaultValue=4, group=ZebraScramble, parameterName=5_ClipSpeed, sortPriority=52}) {name=5_ClipSpeed, nodeColor=<rgba 0 0.5781 1 1>}
  n64 = NamedRerouteUsage()
  NamedRerouteDeclaration(None=Subtract(A=Add(A=@n64), B=NamedRerouteUsage())) {name=5_ClipRange, nodeColor=<rgba 1 0.7734 0 1>}
  NamedRerouteDeclaration(None=ScalarParameter() {defaultValue=3, group=ZebraScramble, parameterName=1_ClipLoop, sortPriority=13}) {name=1_ClipLoop, nodeColor=<rgba 0 0.5781 1 1>}
  n65 = Multiply(A=NamedRerouteUsage(), B=ScalarParameter() {defaultValue=6, group=ZebraScramble, parameterName=5_ClipLoop, sortPriority=53})
  NamedRerouteDeclaration(None=Divide(A=NamedRerouteDeclaration(None=@n65) {name=5_ClipLength, nodeColor=<rgba 0 1 1 1>}, B=NamedRerouteUsage())) {name=5_ClipTimeLength, nodeColor=<rgba 1 0 0.5859 1>}
  n66 = NamedRerouteUsage()
  n67 = NamedRerouteUsage()
  n68 = Subtract(A=NamedRerouteUsage(), B=Multiply(A=NamedRerouteUsage(), B=Divide(A=@n66, B=@n67)))
  n69 = NamedRerouteUsage()
  NamedRerouteDeclaration(None=Add(A=Fmod(A=@n68, B=@n69), B=NamedRerouteUsage())) {name=2_ClipIndex, nodeColor=<rgba 0 0.4609 1 1>}
  n70 = NamedRerouteUsage()
  n71 = NamedRerouteUsage()
  NamedRerouteDeclaration(None=Add(A=Fmod(A=@n70, B=@n71), B=NamedRerouteUsage())) {name=1_ClipIndex, nodeColor=<rgba 0 1 0.0859 1>}
  n72 = NamedRerouteUsage()
  n73 = Divide(A=NamedRerouteUsage(), B=NamedRerouteUsage())
  n74 = NamedRerouteUsage()
  n75 = Divide(A=NamedRerouteUsage(), B=NamedRerouteUsage())
  n76 = Subtract(A=NamedRerouteUsage(), B=Add(A=Multiply(A=@n72, B=@n73), B=Multiply(A=@n74, B=@n75)))
  n77 = NamedRerouteUsage()
  NamedRerouteDeclaration(None=Add(A=Fmod(A=@n76, B=@n77), B=NamedRerouteUsage())) {name=3_ClipIndex, nodeColor=<rgba 0.9688 1 0 1>}
  NamedRerouteDeclaration(None=Multiply(A=NamedRerouteUsage(), B=NamedRerouteUsage())) {name=4_ClipTime, nodeColor=<rgba 1 0 0.8203 1>}
  NamedRerouteDeclaration(None=Multiply(A=NamedRerouteUsage(), B=NamedRerouteUsage())) {name=5_ClipTime, nodeColor=<rgba 1 0 0.8203 1>}
  n78 = Multiply(A=NamedRerouteUsage(), B=Divide(A=NamedRerouteUsage(), B=NamedRerouteUsage()))
  n79 = Multiply(A=NamedRerouteUsage(), B=Divide(A=NamedRerouteUsage(), B=NamedRerouteUsage()))
  n80 = NamedRerouteUsage()
  n81 = Divide(A=NamedRerouteUsage(), B=NamedRerouteUsage())
  n82 = Subtract(A=NamedRerouteUsage(), B=Add(A=Add(A=@n78, B=@n79), B=Multiply(A=@n80, B=@n81)))
  n83 = NamedRerouteUsage()
  NamedRerouteDeclaration(None=Add(A=Fmod(A=@n82, B=@n83), B=NamedRerouteUsage())) {name=4_ClipIndex, nodeColor=<rgba 0.9688 1 0 1>}
  n84 = NamedRerouteUsage()
  n85 = NamedRerouteUsage()
  n86 = NamedRerouteUsage()
  n87 = NamedRerouteUsage()
  n88 = Add(A=Multiply(A=NamedRerouteUsage(), B=Divide(A=@n84, B=@n85)), B=Multiply(A=NamedRerouteUsage(), B=Divide(A=@n86, B=@n87)))
  n89 = Multiply(A=NamedRerouteUsage(), B=Divide(A=NamedRerouteUsage(), B=NamedRerouteUsage()))
  n90 = NamedRerouteUsage()
  n91 = Divide(A=NamedRerouteUsage(), B=NamedRerouteUsage())
  n92 = Subtract(A=NamedRerouteUsage(), B=Add(A=Add(A=@n88, B=@n89), B=Multiply(A=@n90, B=@n91)))
  n93 = NamedRerouteUsage()
  NamedRerouteDeclaration(None=Add(A=Fmod(A=@n92, B=@n93), B=NamedRerouteUsage())) {name=5_ClipIndex, nodeColor=<rgba 0.9688 1 0 1>}
  NamedRerouteDeclaration(None=Add(A=NamedRerouteUsage(), B=NamedRerouteUsage())) {name=3_ClipStartTime, nodeColor=<rgba 0.5703 1 0 1>}
  NamedRerouteDeclaration(None=Add(A=NamedRerouteUsage(), B=NamedRerouteUsage())) {name=4_ClipStartTime, nodeColor=<rgba 0.5703 1 0 1>}
  NamedRerouteDeclaration(None=Add(A=NamedRerouteUsage(), B=NamedRerouteUsage())) {name=5_ClipStartTime, nodeColor=<rgba 0.5703 1 0 1>}
  n94 = NamedRerouteUsage()
  n95 = NamedRerouteUsage()
  n96 = NamedRerouteUsage()
  n97 = NamedRerouteUsage()
  n98 = SphereMask(A=NamedRerouteUsage(), B=NamedRerouteUsage(), Radius=NamedRerouteUsage(), Hardness=NamedRerouteUsage()) {attenuationRadius=0.01, hardnessPercent=90}
  n99 = NamedRerouteUsage()
  n100 = NamedRerouteUsage()
  n101 = NamedRerouteUsage()
  n102 = NamedRerouteUsage()
  n103 = SphereMask(A=NamedRerouteUsage(), B=NamedRerouteUsage(), Radius=NamedRerouteUsage(), Hardness=NamedRerouteUsage()) {attenuationRadius=0.01, hardnessPercent=90}
  n104 = SphereMask(A=NamedRerouteUsage(), B=Constant(), Radius=NamedRerouteUsage(), Hardness=NamedRerouteUsage()) {attenuationRadius=0.01, hardnessPercent=90}
  n105 = Add(
   A=SphereMask(A=NamedRerouteUsage(), B=NamedRerouteUsage(), Radius=NamedRerouteUsage(), Hardness=NamedRerouteUsage()) {attenuationRadius=0.01, hardnessPercent=90}
   B=Add(A=SphereMask(A=@n99, B=@n100, Radius=@n101, Hardness=@n102) {attenuationRadius=0.01, hardnessPercent=90}, B=Add(A=@n103, B=@n104))
  )
  NamedRerouteDeclaration(None=Add(A=SphereMask(A=@n94, B=@n95, Radius=@n96, Hardness=@n97) {attenuationRadius=0.01, hardnessPercent=90}, B=Add(A=@n98, B=@n105))) {name=StaticBlend, nodeColor=<rgba 0.5 1 0 1>}
  n106 = NamedRerouteUsage()
  n107 = Multiply(
   A=ScalarParameter() {defaultValue=4, group=ZebraScramble, parameterName=ZS_FlipbookColumns, sortPriority=2}
   B=ScalarParameter() {defaultValue=4, group=ZebraScramble, parameterName=ZS_FlipbookRows, sortPriority=1}
  )
  NamedRerouteDeclaration(None=Floor(None=Divide(A=@n106, B=@n107) {constB=4}) {desc=sub texture index}) {name=TextureIndex, nodeColor=<rgba 1 0 0.1875 1>}
  n108 = ScalarParameter() {defaultValue=0.5, group=TV Glitches, parameterName=StaticLength}
  n109 = Saturate(None=ScalarParameter() {defaultValue=0.8, group=TV Glitches, parameterName=StaticBlendHardness})
  NamedRerouteDeclaration(None=Multiply(
    A=NamedRerouteDeclaration(None=@n108) {name=StaticLength, nodeColor=<rgba 0 1 0.25 1>}
    B=NamedRerouteDeclaration(None=@n109) {name=StaticBlendHardness, nodeColor=<rgba 1 0 0.5625 1>}
   )) {name=SolidStatic, nodeColor=<rgba 0 1 0.9766 1>}
  n110 = Clamp(None=ScalarParameter() {group=ZebraScramble, parameterName=2_ClipStaticBias, sortPriority=24}) {minDefault=-1}
  n111 = NamedRerouteUsage()
  NamedRerouteDeclaration(None=Add(A=Multiply(A=@n110, B=@n111), B=NamedRerouteUsage())) {name=2_ClipStaticTime, nodeColor=<rgba 1 0.3516 0 1>}
  n112 = Clamp(None=ScalarParameter() {group=ZebraScramble, parameterName=1_ClipStaticBias, sortPriority=14}) {minDefault=-1}
  n113 = NamedRerouteUsage()
  NamedRerouteDeclaration(None=Add(A=Multiply(A=@n112, B=@n113), B=NamedRerouteUsage())) {name=1_ClipStaticTime, nodeColor=<rgba 1 0.3516 0 1>}
  NamedRerouteDeclaration(None=Multiply(A=NamedRerouteUsage(), B=NamedRerouteUsage())) {name=1_ClipTime, nodeColor=<rgba 1 0 0.8203 1>}
  n114 = Clamp(None=ScalarParameter() {group=ZebraScramble, parameterName=3_ClipStaticBias, sortPriority=34}) {minDefault=-1}
  n115 = NamedRerouteUsage()
  NamedRerouteDeclaration(None=Add(A=Multiply(A=@n114, B=@n115), B=NamedRerouteUsage())) {name=3_ClipStaticTime, nodeColor=<rgba 1 0.3516 0 1>}
  n116 = Clamp(None=ScalarParameter() {group=ZebraScramble, parameterName=4_ClipStaticBias, sortPriority=44}) {minDefault=-1}
  n117 = NamedRerouteUsage()
  NamedRerouteDeclaration(None=Add(A=Multiply(A=@n116, B=@n117), B=NamedRerouteUsage())) {name=4_ClipStaticTime, nodeColor=<rgba 1 0.3516 0 1>}
  NamedRerouteDeclaration(None=Floor(None=Fmod(A=@n106, B=@n107)) {desc=uv index in sub texture}) {name=FlipbookSubindex, nodeColor=<rgba 1 0 0.2812 1>}
  NamedRerouteDeclaration(None=LinearInterpolate(A=TextureCoordinate(), B=TextureCoordinate() {coordinateIndex=1}, Alpha=ScalarParameter() {group=ZebraScramble, parameterName=UseUV1})) {name=UVs, nodeColor=<rgba 1 0.4453 0 1>}
  n118 = NamedRerouteUsage()
  NamedRerouteDeclaration(None=Subtract(A=Add(A=@n118), B=NamedRerouteUsage())) {name=1_ClipRange, nodeColor=<rgba 1 0.9609 0 1>}
  n119 = NamedRerouteUsage()
  n120 = If(A=NamedRerouteUsage(), B=NamedRerouteUsage(), "A > B"=NamedRerouteUsage(), "A == B"=@n119, "A < B"=@n119)
  n121 = If(A=NamedRerouteUsage(), B=NamedRerouteUsage(), "A > B"=NamedRerouteUsage(), "A == B"=@n120, "A < B"=@n120)
  n122 = If(A=NamedRerouteUsage(), B=NamedRerouteUsage(), "A > B"=NamedRerouteUsage(), "A == B"=@n121, "A < B"=@n121)
  NamedRerouteDeclaration(None=If(A=NamedRerouteUsage(), B=NamedRerouteUsage(), "A > B"=NamedRerouteUsage(), "A == B"=@n122, "A < B"=@n122)) {name=FlipbookIndex, nodeColor=<rgba 0.6641 1 0 1>}
  NamedRerouteDeclaration(None=ScalarParameter() {defaultValue=20, group=ZebraScramble, parameterName=2_ClipStart, sortPriority=20}) {name=2_ClipStart, nodeColor=<rgba 0 1 0.3672 1>}
  n123 = ScalarParameter() {defaultValue=8, group=ZebraScramble, parameterName=ZS_FlipbookColumns, sortPriority=2}
  n124 = AppendVector(A=@n123, B=ScalarParameter() {defaultValue=8, group=ZebraScramble, parameterName=ZS_FlipbookRows, sortPriority=1})
  n125 = Divide(A=NamedRerouteUsage(), B=@n124)
  n126 = NamedRerouteUsage()
  n127 = Fmod(A=@n126, B=@n123)
  n128 = Divide(A=@n126, B=@n123)
  n129 = Divide(A=AppendVector(A=Floor(None=@n127), B=Floor(None=@n128)), B=@n124)
  TextureSampleParameter2D(UVs=Add(A=Add(A=@n125, B=@n129), B=NamedRerouteUsage())) {
   group=ZebraScramble
   parameterName=ZS_Flipbook
   sortPriority=0
   texture=/BRCosmetics/Accessories/FORT_Backpacks/Backpack_M_MED_ZebraScramble_Bacon/Textures/M_FrenchFry_screen_v3_E
  }
 ]}
