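# Material for the screen on the ZebraScramble Bacon backpack cosmetic, drawing an emissive display on skeletal meshes.
# MF_ZebraScrambleScreen_V2 generates the screen image, which a brightness parameter scales into Emissive.
# The shared character PreFX and PostFX functions bracket a material-attribute override that zeroes base color, metallic, specular and roughness and flattens the normal, leaving the surface lit only by its emissive.
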Material /BRCosmetics/Accessories/FORT_Backpacks/Backpack_M_MED_ZebraScramble_Bacon/Materials/M_ZebraScrambleScreen
graph {
 properties {
  bUseMaterialAttributes=true
  bUsedWithMeshDeformer=true
  bUsedWithSkeletalMesh=true
  pixelDepthOffsetMode=PDOM_Legacy
  usedShadingModels=Default Lit
 }
 nodes [
  n1 = MaterialFunctionCall() {materialFunction=&BaseFN_CharacterFunction_PreFX}
  n2 = Constant()
  n3 = SetMaterialAttributes!(
   MaterialAttributes=@n1.Result
   Base Color=@n2
   Emissive Color=Multiply(
    A=MaterialFunctionCall() {materialFunction=&/MF_ZebraScrambleScreen_V2}.Result
    B=ScalarParameter() {defaultValue=0.66, parameterName=ScreenBrightnessMultiplier}
   ) {constB=0.66}
   Normal=Constant3Vector() {constant=<rgba 0 0 1 0>}
   Metallic=@n2
   Specular=@n2
   Roughness=@n2
  ) {attributeSetTypes [
    #1
    #2
    #3
    #4
    #5
    #6
   ]}
  MaterialFunctionCall(AO=@n1.AO, Distance Blend=@n1.AO, Fresnel=@n1.AO, Gmap=@n1.AO, Pre Skinned Local Position=@n1.AO, Material=@n3) {materialFunction=&BaseFN_CharacterFunction_PostFX}
 ]
}
