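# Animation Blueprint for the ZebraScramble Bacon backpack that repositions the pack during skydive floating and adds physics sway to its antenna.
# A Modify Bone node replaces the pack bone's translation, blended in and out by bIsSkydiveFloating, which this Blueprint does not declare.
# The result feeds two chained Anim Dynamics nodes on the antenna bones, using world-space simulation, angular springs and angular limits.
# The EventGraph has no connected logic.
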AnimBlueprint /BRCosmetics/Accessories/FORT_Backpacks/Backpack_M_MED_ZebraScramble_Bacon/Meshes/M_MED_ZebraScramble_Bacon_Pack_AnimBP
target_skeleton=&/M_MED_ZebraScramble_Bacon_Pack_Skeleton
graphs [
 AnimGraph {nodes [
   n1 = "Transform (Modify) Bone - Bone: pack"!(
    ComponentPose=Local To Component(LocalPose=Input Pose())
    bAlphaBoolEnabled:bool=Get bIsSkydiveFloating()
    Translation="1.1,-2,0"
    Rotation="0,0,0"
    Scale
   ) -> (Pose) {
    alphaInputType=Bool
    translationMode=BMM_Replace
    alphaBoolBlend {blendInTime=0.1, blendOption=QuadraticInOut, blendOutTime=0.1}
    alphaScaleBiasClamp {bInterpResult=true}
    boneToModify {boneName=pack}
    translation=<1.1 -2>
   }
   n2 = Anim Dynamics - Bone: Dyn_Antenna1(ComponentPose=@n1, Alpha=1, ExternalForce="0,0,0") {
    angularBiasOverride=0.9
    angularDampingOverride=0.99
    angularSpringConstant=8
    bAngularSpring=true
    bOverrideAngularBias=true
    bOverrideAngularDamping=true
    bUseGravityOverride=true
    linearDampingOverride=0.8
    linearSpringConstant=1
    simulationSpace=World
    alphaBoolBlend {blendOption=Linear}
    boundBone {boneName=Dyn_Antenna1}
    gravityOverride {z=-980}
    physicsBodyDefinitions [{
      collisionType=CustomSphere
      sphereCollisionRadius=1
      boundBone {boneName=Dyn_Antenna1}
      boxExtents=<1 1 1>
      constraintSetup {angularLimitsMax=<0 10 2.5>, angularLimitsMin=<0 -25 -2.5>, angularTarget=<1 0 0>}
      localJointOffset=<0 0 0>
     }]
   }
   n3 = Anim Dynamics - Bone: Dyn_Antenna2(ComponentPose=@n2, Alpha=1, ExternalForce="0,0,0") {
    angularBiasOverride=0.9
    angularDampingOverride=0.99
    angularSpringConstant=8
    bAngularSpring=true
    bOverrideAngularBias=true
    bOverrideAngularDamping=true
    bUseGravityOverride=true
    linearDampingOverride=0.8
    linearSpringConstant=1
    simulationSpace=World
    alphaBoolBlend {blendOption=Linear}
    boundBone {boneName=Dyn_Antenna2}
    gravityOverride {z=-980}
    physicsBodyDefinitions [{
      collisionType=CustomSphere
      sphereCollisionRadius=1
      boundBone {boneName=Dyn_Antenna2}
      boxExtents=<1 1 1>
      constraintSetup {angularLimitsMax=<0 10 2.5>, angularLimitsMin=<0 -15 -2.5>, angularTarget=<1 0 0>}
      localJointOffset=<0 0 0>
     }]
   }
   Output Pose(Result=Component To Local(ComponentPose=@n3)) {layerGroup=None}
  ]}
 EventGraph {nodes [
   TryGetPawnOwner()
   Event BlueprintUpdateAnimation()
  ]}
]
variables []
